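; Passive player augment ability for Explosive Mag Increase that triggers the augment when the owning pawn searches an allowed type of container.
; On activation it binds to the pawn's server searched-container event and checks the container's gameplay tags against the allowed types.
; A match stores the container's loot spawn location and toss direction, then calls ActivateAugment.
; GetLocationToSpawn returns a randomly jittered point raised above and in front of the container.
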(event Ability|EventFailedToActivatePassiveAbility
  (params (ActorInfo :type Gameplay%20Ability%20Actor%20Info%20Structure)))

(event Ability|EventActivateAbility
  (EventDispatchers|BindEventtoServer_OnSearchedContainer (CallFunction|GetAvatarasFortPlayerPawn) (delegate Custom|OnSearch)))

(event Custom|OnSearch
  (params (SearchingPawn :type Fort%20Pawn%20Object%20Reference) (Container :type Building%20Container%20Object%20Reference))
  (bind _output_get (Variables|Default|SetContainer Container))
  (if (GameplayTags|HasAnyMatchingGameplayTags _output_get (Variables|Default|GetAllowedContainerTypes))
    (Variables|Default|SetSpawnLocation (Variables|/Script/FortniteGame.BuildingContainer|GetLootSpawnLocation _output_get))
    (Variables|Default|SetSpawnDirection (Variables|/Script/FortniteGame.BuildingContainer|GetInstancedLoot_TossDirection _output_get))
    (CallFunction|ActivateAugment)))

(fn GetLocationToSpawn
  (return (Utilities|Operators|Add (Transformation|GetActorLocation (Variables|Default|GetContainer)) (Math|Vector|MakeVector (Math|Random|RandomFloatinRange -25.0 25.0) (Math|Random|RandomFloatinRange -25.0 25.0) 5e+01) (* (Transformation|GetActorForwardVector (Variables|Default|GetContainer)) 5e+01))))
